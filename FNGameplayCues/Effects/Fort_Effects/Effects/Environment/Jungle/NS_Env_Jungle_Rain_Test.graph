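# Test Niagara system for jungle rain: GPU close and mid streaks, mobile mesh rain, splashes, and drips from objects.
# Time-of-day curves gate all rain spawning; player-indoor state cuts splashes and drips, and IndoorAlpha scales opacity.
# Close and mid streaks add scaled local-player velocity and use grid-based scratch modules around the owner.
# Splashes and drips snap to surfaces via depth buffer and distance fields, killing invalid samples.
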NiagaraSystem /FNGameplayCues/Effects/Fort_Effects/Effects/Environment/Jungle/NS_Env_Jungle_Rain_Test
dependencies {
 usedDataInterfaces [
  &NiagaraDataInterfaceCurve
  /Script/Niagara.NiagaraDataInterfaceActorComponent
  /Script/Niagara.NiagaraDataInterfaceTexture
  &NiagaraDataInterfaceColorCurve
  &NiagaraDataInterfaceCollisionQuery
 ]
 usedDynamicInputs [&ColorFromCurve]
 usedModules [
  &SystemState
  &self:SystemSpawnScript.SystemScriptSource.SystemScriptGraph.NiagaraNodeAssignment_0.SetVariables_#1
 ]
}
emitters {
 FN_DebugGnomon {
  bEnabled=false
  emitterUpdateScript [
   &EmitterState(
    Life Cycle Mode=::NewEnumerator1
    Loop Behavior=::NewEnumerator1
    Loop Duration Mode=::NewEnumerator1
    Scalability Mode=::NewEnumerator1
    UseLoopDelay=false
   )
   &SpawnBurst_Instantaneous(Use Spawn Probability=false, UseLoopCountLimit=false)
  ]
  particleSpawnScript [&InitializeParticle(
    Color Mode=::NewEnumerator0
    Lifetime=5
    Mesh Scale Mode=::NewEnumerator3
    Mesh Uniform Scale=5
    UsePositionOffset=false
    Write Material Random=true
   )]
  particleUpdateScript [&ParticleState(Kill Particles When Lifetime Has Elapsed=false)]
  rendererClasses [&NiagaraMeshRendererProperties]
  renderers [{rendererClass=&NiagaraMeshRendererProperties, rendererIndex=0}]
 }
 RainFromObjects001 {
  simTarget=GPUComputeSim
  emitterUpdateScript [
   &EmitterState()
   &SpawnRate(
    SpawnRate=&Multiply_Float(
     A=&FloatFromCurve(
      FloatCurve=DataInterface() {propertyValues {Curve {keys [<at 0.2 1>,<at 0.3 0>,<at 0.95 0>,<at 1 1>]}}}
      CurveIndex=Linked() {linkedVariable {name=NPC.FortniteNPC.TimeOfDay_0-1, type=NiagaraFloat}}
      Scale Curve=$User.Rain_SpawnRateScale
     )
     B=&Multiply_Float(A=&Lerp_Float(A=1, B=0, Alpha=$NPC.FortniteNPC.IsPlayerIndoors), B=10000)
    )
    Use Spawn Probability=false
   )
  ]
  particleSpawnScript [
   &InitializeParticle(
    Lifetime Mode=::NewEnumerator1
    Sprite Rotation Mode=::NewEnumerator1
    Sprite Size Max=<25 20>
    Sprite Size Min=<25 15>
    Sprite Size Mode=::NewEnumerator2
    UsePositionOffset=false
    Write Material Random=true
   ) {module=InitializeParticle001}
   &AddVelocity(Velocity=&RandomRangeVector(Minimum=<0 0 -25>, Maximum=<0 0 -50>))
   &SphereLocation(
    MaskSpawnGroup=false
    Non Uniform Scale=<1 1 0.7>
    Offset=<0 0 1500>
    Override Seed=false
    Sphere Radius=2500
    Surface Only=false
    Surface Only Band Thickness=0.5
    UseNonUniformScale=true
    UseOffset=true
    UseSphereOrientation=false
   )
   /Niagara/Modules/Math/MoveToNearestDistanceFieldSurface_GPU(Kill by Distance=800, bKillByDistance=true)
   /SetVariables(Particles.InitialPos:NiagaraPosition=$Particles.Position) {isSetParameters=true}
   /Niagara/Modules/Math/PlaceParticlesOnDepthBuffer_GPU(Max Z Depth=1500, Module.Use Max Z Depth=true)
   &KillParticles(Kill Particles=&SetBoolByFloatComparison(
     A=&VectorLength(Vector=&Subtract_Vector(A=$Particles.InitialPos, B=Linked() {linkedVariable {name=Particles.Position, type=NiagaraPosition}}))
     B=1500
    ))
   &KillParticles(Kill Particles=&Boolean_NotOperation(Bool=$Output.PlaceParticlesOnDepthBuffer_GPU.ValidSample)) {module=KillParticles001}
   &SubUVAnimation(End Frame=3, SubUV Animation Mode:ENiagara_SubUVLookupMode=::NewEnumerator2)
   &CameraOffset(Camera Offset Amount=100)
   &self:ScratchModule_13(Angle=180, BlendDirection=1, Normal:Vector3f=<0 0 1>, PickCeilingOrFloor=true, Velocity=-200) {module=ScratchModule_13}
  ]
  particleUpdateScript [
   &ParticleState()
   &RegionMask(Inner Float=0, Outer Float=2, Region Falloff=0, Region Radius:NiagaraFloat=$User.Rain_MaskInside_Radius, UseOffset=false, UseRotation=false) {module=RegionMask001}
   &ScaleSpriteSize(Uniform Curve Sprite Scale=DataInterface() {propertyValues {Curve {keys [{tangentMode=RCTM_User, time=0, value=0.85}, {tangentMode=RCTM_User, time=1, value=1}]}}})
   &GravityForce(Gravity=<0 0 -800>)
   &Drag(
    Drag=&FloatFromCurve(FloatCurve=DataInterface() {propertyValues {Curve {keys [{tangentMode=RCTM_User, time=0.25, value=0}, {tangentMode=RCTM_User, time=1, value=1}]}}})
    Use Linear Drag=true
    Use Rotational Drag=true
   )
   &SolveForcesAndVelocity(Clamp Velocity=true, Limit Acceleration=false, Speed Limit=5000)
   &ScaleSpriteSizeBySpeed(
    Max Scale Factor=<0.1 2>
    Min Scale Factor=<0.2 0.25>
    Sample Scale Factor By Curve=true
    Source Velocity:Vector3f=&Multiply_Vector(A=<0.05 0.05 1>, B=$Particles.Velocity)
    Velocity Threshold=500
   )
   &ScaleColor(
    Scale Alpha=&Multiply_Float(
     A=&Multiply_Float(
      A=&Multiply_Float(
       A=&FloatFromCurve(
        FloatCurve=DataInterface() {propertyValues {Curve {keys [
            {interpMode=RCIM_Cubic, tangentMode=RCTM_User, time=0, value=0}
            {interpMode=RCIM_Cubic, tangentMode=RCTM_User, time=0.5, value=1}
            {tangentMode=RCTM_User, time=1, value=0}
           ]}}}
        Scale Curve=0.5
       )
       B=$System.IndoorAlpha
      )
      B=$Particles.RegionMask001.Float
     )
     B=$User.Rain_AlphaMult
    )
    Scale RGB=&MakeVectorFromLinearColorRGB(Color=&MultiplyLinearColor(Linear Color=$User.Rain_NonLitColor, Scale Factor=&MakeVector4_FromFloat(Float=$User.Rain_ColorMult)))
    ScaleA=true
    ScaleRGB=true
   )
  ]
  rendererClasses [&NiagaraSpriteRendererProperties]
  renderers [{rendererClass=&NiagaraSpriteRendererProperties}]
 }
 Rain_Close {
  simTarget=GPUComputeSim
  emitterUpdateScript [
   &EmitterState(Life Cycle Mode=::NewEnumerator1, Loop Duration=2, UseLoopDelay=true)
   &SpawnRate(SpawnRate=5000, Use Spawn Probability=false) {enabled=false, module=SpawnRate001}
   &SpawnRate(
    SpawnRate=&Multiply_Float(
     A=&FloatFromCurve(
      FloatCurve=DataInterface() {propertyValues {Curve {keys [<at 0.08 1>,<at 0.185 0>,<at 0.9375 0>,<at 1 1>]}}}
      CurveIndex=Linked() {linkedVariable {name=NPC.FortniteNPC.TimeOfDay_0-1, type=NiagaraFloat}}
      Scale Curve=$User.Rain_SpawnRateScale
     )
     B=5000
    )
    Use Spawn Probability=false
   )
   /SetVariables(Emitter.VolumeRadius=250) {isSetParameters=true}
  ]
  particleSpawnScript [
   &InitializeParticle(
    Lifetime Max=1
    Lifetime Min=0.6
    Lifetime Mode=::NewEnumerator1
    Sprite Size Max=<1 1>
    Sprite Size Min=<1 1>
    Sprite Size Mode=::NewEnumerator2
    UsePositionOffset=false
    Write Material Random=true
   ) {module=InitializeParticle001}
   &self:ScratchModule_08(New Actor Component Interface=DataInterface() {propertyValues {
      LocalPlayerIndex=0
      SourceActorRef=null
      SourceMode=LocalPlayer
      bRequireCurrentFrameData=true
      ActorOrComponentParameter {parameter {name=None, typeDefHandle {registeredTypeIndex=1}, varData []}}
     }}) {module=ScratchModule_08}
   &AddVelocity(
    Coordinate Space=::World
    Velocity=&Add_Vector(
     A=&Multiply_VectorByFloat(Vector=&RandomRangeVector(Minimum=<-50 0 -800>, Maximum=<-100 0 -1000>), Float=$User.Rain_FallingSpeed)
     B=&Multiply_Vector(A=$Output.ScratchModule_08.Velocity, B=$User.Rain_InheritVel)
    )
   )
   &ShapeLocation(MaskSpawnGroup=false, Override Seed=false, Sphere Radius=5000, Sphere Surface Distribution=1, UseOwnerScale=true)
  ]
  particleUpdateScript [
   &ParticleState()
   &WindForce(Wind Speed=&Lerp_Vector(A=<600 0 0>, B=<800 0 0>, Alpha=$System.StormIntensity)) {enabled=false, module=WindForce001}
   &SolveForcesAndVelocity(Clamp Velocity=false, Limit Acceleration=false)
   /SetVariables(Particles.PositionCenter:NiagaraPosition=&AddVectorToPosition(Position=$Engine.Owner.Position, Vector=$Emitter.OriginOffset)) {isSetParameters=true}
   &self:ScratchModule_11(
    GridSize:NiagaraFloat=&Multiply_Float(A=$Emitter.VolumeRadius, B=10)
    InputPosition:NiagaraPosition=$Particles.PositionCenter
   ) {module=ScratchModule_11}
   &ScaleSpriteSizeBySpeed(
    Max Scale Factor=<3 80>
    Min Scale Factor=<2 50>
    Source Velocity:Vector3f=&Multiply_Vector(A=<0.05 0.05 1>, B=$Particles.Velocity)
   )
   &AlignSpriteToMeshOrientation(SetAlignment=true, SetFacing=true)
   &self:ScratchModule_05(
    CaptureMatrix:NiagaraMatrix=$User.CaptureMatrix
    CaptureOrthoWidth:NiagaraFloat=$User.CaptureOrthoWidth
    CaptureTexture=DataInterface() {propertyValues {
      Texture=/Game/Effects/Fort_Effects/Effects/Environment/Jungle/Textures/RainMask
      TextureUserParameter {parameter {name=None, typeDefHandle {registeredTypeIndex=2}, varData []}}
     }}
    SamplePosition:NiagaraPosition=$Particles.Position
   ) {enabled=false, module=ScratchModule_05}
   &KillParticlesInVolume(Box Size:Vector3f=<2500 2500 15000>, Kill Shape=::NewEnumerator1) {enabled=false}
   &RegionMask(Inner Float=0, Offset:Vector3f=<0 0 -3000>, Outer Float=1, Region Radius=2500, UseOffset=true, UseRotation=false)
   &ScaleColor(
    Scale Alpha=&Multiply_Float(
     A=&Multiply_Float(
      A=&FloatFromCurve(
       FloatCurve=DataInterface() {propertyValues {Curve {keys [
           {interpMode=RCIM_Cubic, tangentMode=RCTM_User, time=0, value=0}
           {interpMode=RCIM_Cubic, tangentMode=RCTM_User, time=0.5, value=1}
           {tangentMode=RCTM_User, time=1, value=0}
          ]}}}
       Scale Curve=$Particles.RegionMask.Float
      )
      B=$System.IndoorAlpha
     )
     B=$User.Rain_AlphaMult
    )
    Scale RGB=&MakeVectorFromLinearColorRGB(Color=&MultiplyLinearColor(Linear Color=$User.Rain_NonLitColor, Scale Factor=&MakeVector4_FromFloat(Float=$User.Rain_ColorMult)))
    ScaleA=true
    ScaleRGB=true
   )
  ]
  rendererClasses [&NiagaraSpriteRendererProperties]
  renderers [{rendererClass=&NiagaraSpriteRendererProperties}]
 }
 Rain_Mid {
  simTarget=GPUComputeSim
  emitterUpdateScript [
   &EmitterState(Life Cycle Mode=::NewEnumerator1, Loop Duration=2, UseLoopDelay=true)
   &SpawnRate(
    SpawnRate=&Multiply_Float(
     A=&FloatFromCurve(
      FloatCurve=DataInterface() {propertyValues {Curve {keys [<at 0.08 1>,<at 0.185 0>,<at 0.9375 0>,<at 1 1>]}}}
      CurveIndex=Linked() {linkedVariable {name=NPC.FortniteNPC.TimeOfDay_0-1, type=NiagaraFloat}}
      Scale Curve=$User.Rain_SpawnRateScale
     )
     B=5000
    )
    Use Spawn Probability=false
   )
   /SetVariables(Emitter.VolumeRadius=250) {isSetParameters=true}
  ]
  particleSpawnScript [
   &InitializeParticle(
    Lifetime Max=0.8
    Lifetime Min=0.5
    Lifetime Mode=::NewEnumerator1
    Sprite Size Max=<3.75 4.5>
    Sprite Size Min=<3 3.3>
    Sprite Size Mode=::NewEnumerator2
    UsePositionOffset=false
    Write Material Random=true
   ) {module=InitializeParticle001}
   &self:ScratchModule_09(New Actor Component Interface=DataInterface() {propertyValues {
      LocalPlayerIndex=0
      SourceActorRef=null
      SourceMode=LocalPlayer
      bRequireCurrentFrameData=true
      ActorOrComponentParameter {parameter {name=None, typeDefHandle {registeredTypeIndex=1}, varData []}}
     }}) {module=ScratchModule_09}
   &AddVelocity(
    Coordinate Space=::World
    Velocity=&Add_Vector(
     A=&Multiply_VectorByFloat(
      Vector=&RandomRangeVector(Minimum=<-100 0 -1200>, Maximum=<-250 0 -1400>)
      Float=$User.Rain_FallingSpeed
     )
     B=&Multiply_Vector(A=$Output.ScratchModule_09.Velocity, B=$User.Rain_InheritVel)
    )
   )
   &Color(Color=&MakeLinearColorFromVectorAndFloat("Vector (RGB)"=<0.6281 0.5529 0.974>, "Float (Alpha)"=&ClampFloat(Float=1.2))) {enabled=false}
   &KillParticlesInVolume(Kill Volume Enabled=false, Origin Offset:Vector3f=<0 0 -100>, Sphere Radius=5000) {enabled=false}
   &ShapeLocation(
    Cylinder Height=15000
    Cylinder Radius=6000
    MaskSpawnGroup=false
    Non Uniform Scale=<1 1 0.2>
    Offset=<0 0 1250>
    Offset Mode=::NewEnumerator0
    Override Seed=false
    Shape Primitive=::NewEnumerator1
    Surface Only=true
    Surface Only Band Thickness=0.5
    Use Endcaps In Surface Only Mode=false
    UseOwnerScale=true
   )
  ]
  particleUpdateScript [
   &ParticleState()
   &SolveForcesAndVelocity(Clamp Velocity=false, Limit Acceleration=false)
   /SetVariables(Particles.PositionCenter:NiagaraPosition=&AddVectorToPosition(Position=$Engine.Owner.Position, Vector=$Emitter.OriginOffset)) {isSetParameters=true}
   &self:ScratchModule_11(
    GridSize:NiagaraFloat=&Multiply_Float(A=$Emitter.VolumeRadius, B=50)
    InputPosition:NiagaraPosition=$Particles.PositionCenter
   ) {module=ScratchModule_11}
   &AlignSpriteToMeshOrientation(SetAlignment=true, SetFacing=true)
   &ScaleSpriteSizeBySpeed(
    Max Scale Factor=<1.25 50>
    Min Scale Factor=<1 35>
    Source Velocity:Vector3f=&Multiply_Vector(A=<0.05 0.05 1>, B=$Particles.Velocity)
   )
   &self:ScratchModule_05(
    CaptureMatrix:NiagaraMatrix=$User.CaptureMatrix
    CaptureOrthoWidth:NiagaraFloat=$User.CaptureOrthoWidth
    CaptureTexture=DataInterface() {propertyValues {
      Texture=&RainMask
      TextureUserParameter {parameter {name=None, typeDefHandle {registeredTypeIndex=2}, varData []}}
     }}
    SamplePosition:NiagaraPosition=$Particles.Position
   ) {enabled=false, module=ScratchModule_05}
   &KillParticlesInVolume(Box Size:Vector3f=<5000 5000 5000>, Kill Shape=::NewEnumerator1) {enabled=false, module=KillParticlesInVolume001}
   &RegionMask(Inner Float=0, Offset:Vector3f=<0 0 -2500>, Outer Float=1, Region Radius=5000, UseOffset=true, UseRotation=false)
   &ScaleColor(
    Scale Alpha=&Multiply_Float(
     A=&Multiply_Float(
      A=&FloatFromCurve(
       FloatCurve=DataInterface() {propertyValues {Curve {keys [
           {interpMode=RCIM_Cubic, tangentMode=RCTM_User, time=0, value=0}
           {interpMode=RCIM_Cubic, tangentMode=RCTM_User, time=0.5, value=1}
           {tangentMode=RCTM_User, time=1, value=0}
          ]}}}
       Scale Curve=$Particles.RegionMask.Float
      )
      B=$System.IndoorAlpha
     )
     B=$User.Rain_AlphaMult
    )
    Scale RGB=&Multiply_VectorByFloat(
     Vector=&MakeVectorFromLinearColorRGB(Color=&MultiplyLinearColor(Linear Color=$User.Rain_NonLitColor, Scale Factor=&MakeVector4_FromFloat(Float=$User.Rain_ColorMult)))
     Float=5
    )
    ScaleA=true
    ScaleRGB=true
   )
  ]
  rendererClasses [&NiagaraSpriteRendererProperties]
  renderers [{rendererClass=&NiagaraSpriteRendererProperties}]
 }
 Rain_Mobile {
  simTarget=GPUComputeSim
  emitterUpdateScript [
   &EmitterState(Life Cycle Mode=::NewEnumerator1, Loop Duration=2, UseLoopDelay=true)
   &SpawnRate(
    SpawnRate=&Multiply_Float(
     A=&FloatFromCurve(
      FloatCurve=DataInterface() {propertyValues {Curve {keys [<at 0.08 1>,<at 0.185 0>,<at 0.9375 0>,<at 1 1>]}}}
      CurveIndex=Linked() {linkedVariable {name=NPC.FortniteNPC.TimeOfDay_0-1, type=NiagaraFloat}}
      Scale Curve=$User.Rain_SpawnRateScale
     )
     B=500
    )
    Use Spawn Probability=false
   )
  ]
  particleSpawnScript [
   &InitializeParticle(
    Lifetime Max=1.5
    Lifetime Mode=::NewEnumerator1
    Mesh Scale Max=<10 10 25>
    Mesh Scale Min=<5 5 25>
    Mesh Scale Mode=::NewEnumerator2
    Sprite Size Max=<3.75 4.5>
    Sprite Size Min=<3 3.3>
    Sprite Size Mode=::NewEnumerator2
    UsePositionOffset=false
    Write Material Random=true
   ) {module=InitializeParticle001}
   &AddVelocity(
    Coordinate Space=::World
    Velocity=&Add_Vector(
     A=&Multiply_VectorByFloat(Vector=&RandomRangeVector(Minimum=<0 0 -800>, Maximum=<0 0 -1200>), Float=$User.Rain_FallingSpeed)
     B=<0 0 1>
    )
   )
   /SetVariables() {isSetParameters=true}
   &Color(Color=&MakeLinearColorFromVectorAndFloat("Vector (RGB)"=<0.6281 0.5529 0.974>, "Float (Alpha)"=&ClampFloat(Float=1.2))) {enabled=false}
   &ShapeLocation(
    Cylinder Height=5000
    Cylinder Radius=6000
    MaskSpawnGroup=false
    Non Uniform Scale=<1 1 0.2>
    Offset=<0 0 1250>
    Offset Mode=::NewEnumerator0
    Override Seed=false
    Shape Primitive=::NewEnumerator1
    Surface Only=true
    Surface Only Band Thickness=0.7
    Use Endcaps In Surface Only Mode=false
    UseOwnerScale=true
   )
   &InitialMeshOrientation(Rotation=&RandomRangeVector(Minimum=<0 0 0>, Maximum=<0 0 1>), Use Rotation Vector=true)
  ]
  particleUpdateScript [
   &ParticleState()
   &GravityForce(Gravity=<0 0 -300>)
   &SolveForcesAndVelocity(Clamp Velocity=true, Limit Acceleration=false, Speed Limit=5000)
   &AlignSpriteToMeshOrientation(SetAlignment=true, SetFacing=true)
   &ScaleColor(
    Scale Alpha=&Multiply_Float(
     A=&Multiply_Float(
      A=&FloatFromCurve(
       FloatCurve=DataInterface() {propertyValues {Curve {keys [
           {interpMode=RCIM_Cubic, tangentMode=RCTM_User, time=0, value=0}
           {interpMode=RCIM_Cubic, tangentMode=RCTM_User, time=0.5, value=1}
           {tangentMode=RCTM_User, time=1, value=0}
          ]}}}
       Scale Curve=0.35
      )
      B=$System.IndoorAlpha
     )
     B=$User.Rain_AlphaMult
    )
    Scale RGB=&Multiply_VectorByFloat(Vector=<1 1 1>, Float=$User.Rain_ColorMult)
    ScaleA=true
    ScaleRGB=true
   )
   &DynamicMaterialParameters(
    Index 0 Param 1=&Multiply_Float(
     A=&RandomRangeFloat(Minimum=0.7, Maximum=2)
     B=&FloatFromCurve(FloatCurve=DataInterface() {propertyValues {Curve {keys [
          {interpMode=RCIM_Cubic, tangentMode=RCTM_User, time=0, value=1}
          {interpMode=RCIM_Cubic, tangentMode=RCTM_User, time=0.5, value=1}
          {tangentMode=RCTM_User, time=1, value=1}
         ]}}})
    )
    Index 0 Param 2=&RandomRangeFloat(Minimum=1, Maximum=2)
    Index 0 Param 3=&RandomRangeFloat(Minimum=0.1, Maximum=0.2)
    Module.Param0WriteEnabled=true
    Module.Param1WriteEnabled=true
    Module.Param2WriteEnabled=true
    Module.Param3WriteEnabled=true
   )
  ]
  rendererClasses [&NiagaraMeshRendererProperties]
  renderers [{rendererClass=&NiagaraMeshRendererProperties, rendererIndex=0}]
 }
 SplashesClose001 {
  simTarget=GPUComputeSim
  emitterUpdateScript [
   &EmitterState(
    Enable Distance Culling=true
    MaxDistance=50000
    Module.bApplySpawnCountScaleByDistance=true
    Module.bUseMaxDistance=true
    Module.bUseMinDistance=false
    Scalability Mode=::NewEnumerator1
    Scale Spawn Count=true
    Spawn Count Scale By Distance=DataInterface() {propertyValues {Curve {keys [<at 0 1>,<at 0.25 1>,<at 1 0>]}}}
   )
   &SpawnRate(
    SpawnRate=&Multiply_Float(
     A=&FloatFromCurve(
      FloatCurve=DataInterface() {propertyValues {Curve {keys [<at 0.08 1>,<at 0.185 0>,<at 0.9375 0>,<at 1 1>]}}}
      CurveIndex=Linked() {linkedVariable {name=NPC.FortniteNPC.TimeOfDay_0-1, type=NiagaraFloat}}
      Scale Curve=$User.Rain_SpawnRateScale
     )
     B=&Multiply_Float(A=&Lerp_Float(A=1, B=0, Alpha=$NPC.FortniteNPC.IsPlayerIndoors), B=5000)
    )
    Use Spawn Probability=false
   )
  ]
  particleSpawnScript [
   &InitializeParticle(
    Lifetime=&RandomRangeFloat(Minimum=0.35, Maximum=0.6)
    Sprite Rotation Mode=::NewEnumerator2
    Sprite Size Mode=::NewEnumerator1
    Uniform Sprite Size Max=1
    Uniform Sprite Size Min=1
    UsePositionOffset=false
    Write Material Random=true
   )
   &AddVelocity(Velocity=&RandomRangeVector(Minimum=<0 0 25>, Maximum=<0 0 50>))
   &ShapeLocation(
    Cylinder Height=50
    Cylinder Radius=2000
    MaskSpawnGroup=false
    Offset=<0 0 -50>
    Offset Mode=::NewEnumerator0
    Override Seed=false
    Rotation Mode=::NewEnumerator5
    Shape Primitive=::NewEnumerator1
    Surface Only=true
    Surface Only Band Thickness=0.9
    Use Endcaps In Surface Only Mode=false
    UseOwnerScale=true
   )
   /SetVariables(Particles.InitialPos:NiagaraPosition=$Particles.Position) {isSetParameters=true}
   &PlaceParticlesOnDepthBuffer_GPU(Max Z Depth=250, Module.Use Max Z Depth=true)
   &MoveToNearestDistanceFieldSurface_GPU(Kill by Distance=150, bKillByDistance=true)
   &KillParticles(Kill Particles=&Boolean_NotOperation(Bool=$Output.PlaceParticlesOnDepthBuffer_GPU.ValidSample)) {module=KillParticles001}
   /SetVariables(Particles.SpriteSize=&Vector2DFromFloat(Value=&RandomRangeFloat(
      Minimum=25
      Maximum=&UniformAOrBFloat(A=35, B=&RandomRangeFloat(Minimum=40, Maximum=50), Distribution Weight=0.1)
     ))) {isSetParameters=true}
   &CameraOffset()
   &SubUVAnimation(End Frame=3, SubUV Animation Mode:ENiagara_SubUVLookupMode=::NewEnumerator2)
   &self:ScratchModule_12(Angle=50, BlendDirection=25, Normal:Vector3f=<0 0 1>, Velocity=50) {module=ScratchModule_12}
  ]
  particleUpdateScript [
   &ParticleState()
   &RegionMask(
    Inner Float=0
    Outer Float=2
    Region Falloff=0.1
    Region Radius:NiagaraFloat=$User.Rain_MaskInside_Radius
    UseOffset=false
    UseRotation=false
   ) {module=RegionMask001}
   &ScaleSpriteSize(
    Uniform Curve Scale=1.5
    Uniform Curve Sprite Scale=DataInterface() {propertyValues {Curve {keys [
        {arriveTangent=2.019, interpMode=RCIM_Cubic, leaveTangent=2.019, tangentMode=RCTM_User, time=0, value=0}
        {arriveTangent=-0.0405, interpMode=RCIM_Cubic, leaveTangent=-0.0405, tangentMode=RCTM_User, time=1, value=1}
       ]}}}
   )
   &GravityForce(Gravity=<0 0 -400>)
   &DynamicMaterialParameters(
    Index 0 Param 1=&FloatFromCurve(
     FloatCurve=DataInterface() {propertyValues {Curve {keys [
         {arriveTangent=9.0725, interpMode=RCIM_Cubic, leaveTangent=9.0725, tangentMode=RCTM_User, time=0, value=0}
         {arriveTangent=-0.0783, interpMode=RCIM_Cubic, leaveTangent=-0.0783, tangentMode=RCTM_User, time=0.1408, value=0.9847}
         {arriveTangent=-0.0405, interpMode=RCIM_Cubic, leaveTangent=-0.0405, tangentMode=RCTM_User, time=1, value=0.25}
        ]}}}
     Scale Curve=2
    )
    Index 0 Param 2=0.1
    Index 0 Param 3=&RandomRangeFloat(Minimum=0, Maximum=0.5)
    Index 0 Param 4=1
    Module.Param0WriteEnabled=true
    Module.Param1WriteEnabled=false
    Module.Param2WriteEnabled=false
    Module.Param3WriteEnabled=false
   )
   &SolveForcesAndVelocity(Clamp Velocity=true, Limit Acceleration=false, Speed Limit=550)
   &ScaleColor(
    Scale Alpha=&Multiply_Float(
     A=&Multiply_Float(
      A=&Multiply_Float(
       A=&FloatFromCurve(FloatCurve=DataInterface() {propertyValues {Curve {keys [
            {interpMode=RCIM_Cubic, tangentMode=RCTM_User, time=0, value=0}
            {arriveTangent=-1.4251, interpMode=RCIM_Cubic, leaveTangent=-1.4251, tangentMode=RCTM_User, time=0.1, value=1}
            {tangentMode=RCTM_User, time=1, value=0}
           ]}}})
       B=$System.IndoorAlpha
      )
      B=$Particles.RegionMask001.Float
     )
     B=$User.Rain_AlphaMult
    )
    Scale RGB=&MakeVectorFromLinearColorRGB(Color=&MultiplyLinearColor(Linear Color=$User.Rain_NonLitColor, Scale Factor=&MakeVector4_FromFloat(Float=$User.Rain_ColorMult)))
    ScaleA=true
    ScaleRGB=true
   )
  ]
  rendererClasses [&NiagaraSpriteRendererProperties]
  renderers [{rendererClass=&NiagaraSpriteRendererProperties}]
 }
}
summary {
 systemName=NS_Env_Jungle_Rain_Test
 emitters [
  {emitterName=Rain_Close, simTarget=GPUComputeSim, rendererClasses [&NiagaraSpriteRendererProperties]}
  {emitterName=Rain_Mid, simTarget=GPUComputeSim, rendererClasses [&NiagaraSpriteRendererProperties]}
  {emitterName=SplashesClose001, simTarget=GPUComputeSim, rendererClasses [&NiagaraSpriteRendererProperties]}
  {emitterName=RainFromObjects001, simTarget=GPUComputeSim, rendererClasses [&NiagaraSpriteRendererProperties]}
  {emitterName=Rain_Mobile, simTarget=GPUComputeSim, rendererClasses [&NiagaraMeshRendererProperties]}
  {bEnabled=false, emitterName=FN_DebugGnomon, rendererClasses [&NiagaraMeshRendererProperties]}
 ]
 userVariables [
  User.Landscape {
   type=NiagaraDataInterfaceLandscape
   description=""
   defaultValue {
    struct=NiagaraExt_VariableValue_DataInterface
    value {
     dataInterface=&self:NiagaraDataInterfaceLandscape_0
     dataInterfaceClass=&NiagaraDataInterfaceLandscape
    }
   }
  }
  User.Rain_AlphaMult {type=NiagaraFloat, defaultValue=0.1, description=""}
  User.Rain_ColorMult {type=NiagaraFloat, defaultValue=0.5, description=""}
  User.Rain_FallingSpeed {type=NiagaraFloat, defaultValue=1.5, description=""}
  User.Rain_InheritVel {type=Vector3f, description="", defaultValue {struct=Vector3f, value=<-1 -1 0>}}
  User.Rain_MaskInside_Radius {type=NiagaraFloat, defaultValue=300, description=""}
  User.Rain_NonLitColor {type=LinearColor, description="", defaultValue {struct=LinearColor, value=<rgba 0.8646 0.8886 1 1>}}
  User.Rain_SpawnRateScale {type=NiagaraFloat, defaultValue=0.2, description=""}
 ]
}
